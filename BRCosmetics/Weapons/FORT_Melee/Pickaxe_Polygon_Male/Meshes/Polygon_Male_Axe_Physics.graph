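# Physics asset for the Polygon Male pickaxe axe, a two-handed melee cosmetic with feathers that can be physically simulated.
# One large capsule covers the handle bone, and one small capsule covers each of the big and small feather bones.
# Every feather is constrained directly to the handle in a star layout, with no feather-to-feather chains.
# Swing and twist limits are uniform, except for a tighter limit on feather_small_tpLfA, unlike its right-side counterpart.
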PhysicsAsset /BRCosmetics/Weapons/FORT_Melee/Pickaxe_Polygon_Male/Meshes/Polygon_Male_Axe_Physics
bodies {
 Melee_TwoHanded_Handle [{
   shapeType=Capsule
   length=71.0551
   radius=31.8165
   shapeName=Melee_TwoHanded_Handle_capsule
   center=<0.8683 6.9313 24.0018>
   rotation=<rot 4.6006 -90 -0.0562>
  }]
 feather_big_btLfA [{
   shapeType=Capsule
   length=11.0543
   radius=3.0566
   shapeName=feather_big_btLfA_capsule
   center=<0.0135 -9.2013 -1.1411>
   rotation=<rot 0 -169.1082 -85.741>
  }]
 feather_big_btRtA [{
   shapeType=Capsule
   length=11.1441
   radius=3.0658
   shapeName=feather_big_btRtA_capsule
   center=<-1.396 -9.8531 -0.2831>
   rotation=<rot 0 172.6647 -73.7854>
  }]
 feather_big_mdCtA [{
   shapeType=Capsule
   length=14.3891
   radius=4.5249
   shapeName=feather_big_mdCtA_capsule
   center=<1.7267 -10.8188 1.9178>
   rotation=<rot 0 5.6415 -77.1574>
  }]
 feather_big_mdLfA [{
   shapeType=Capsule
   length=12.47
   radius=3.3333
   shapeName=feather_big_mdLfA_capsule
   center=<1.8474 -8.5028 -1.2277>
   rotation=<rot 0 -173.763 -87.138>
  }]
 feather_big_mdLfB [{
   shapeType=Capsule
   length=12.3869
   radius=4.024
   shapeName=feather_big_mdLfB_capsule
   center=<1.9322 -9.8358 1.1019>
   rotation=<rot 0 24.6086 -73.7967>
  }]
 feather_big_mdRtA [{
   shapeType=Capsule
   length=11.5103
   radius=3.6116
   shapeName=feather_big_mdRtA_capsule
   center=<-0.2499 -9.8011 1.8682>
   rotation=<rot 0 1.498 -76.1931>
  }]
 feather_big_mdRtB [{
   shapeType=Capsule
   length=11.7533
   radius=3.9448
   shapeName=feather_big_mdRtB_capsule
   center=<-1.9643 -9.5889 3.1086>
   rotation=<rot 0 -23.4102 -71.998>
  }]
 feather_big_tpLfA [{
   shapeType=Capsule
   length=16.7319
   radius=5.1188
   shapeName=feather_big_tpLfA_capsule
   center=<1.1768 -10.5893 5.8746>
   rotation=<rot 0 8.2827 -74.7975>
  }]
 feather_big_tpRtA [{
   shapeType=Capsule
   length=15.3386
   radius=5.0564
   shapeName=feather_big_tpRtA_capsule
   center=<-1.613 -13.2392 5.4861>
   rotation=<rot 0 -9.0477 -70.0287>
  }]
 feather_small_btLfA [{
   shapeType=Capsule
   length=9.9211
   radius=2.9384
   shapeName=feather_small_btLfA_capsule
   center=<3.1101 -8.2899 -2.7098>
   rotation=<rot 0 -155.9522 -86.2667>
  }]
 feather_small_btLfB [{
   shapeType=Capsule
   length=6.3791
   radius=2.4791
   shapeName=feather_small_btLfB_capsule
   center=<1.8067 -5.0216 -2.4795>
   rotation=<rot 0 -147.08 -62.4399>
  }]
 feather_small_btRtA [{
   shapeType=Capsule
   length=7.5557
   radius=2.3406
   shapeName=feather_small_btRtA_capsule
   center=<-3.1603 -4.9729 -2.4774>
   rotation=<rot 0 140.6297 -69.4447>
  }]
 feather_small_btRtB [{
   shapeType=Capsule
   length=8.8076
   radius=2.5172
   shapeName=feather_small_btRtB_capsule
   center=<-3.1368 -7.6936 -1.557>
   rotation=<rot 0 158.3888 -85.8232>
  }]
 feather_small_mdLfA [{
   shapeType=Capsule
   length=10.4687
   radius=4.0938
   shapeName=feather_small_mdLfA_capsule
   center=<3.2353 -8.8039 1.1066>
   rotation=<rot 0 19.562 -77.9599>
  }]
 feather_small_mdLfB [{
   shapeType=Capsule
   length=9.1005
   radius=3.5577
   shapeName=feather_small_mdLfB_capsule
   center=<2.0489 -5.9391 5.6969>
   rotation=<rot 0 15.0328 -49.3858>
  }]
 feather_small_mdRtA [{
   shapeType=Capsule
   length=8.4163
   radius=3.6363
   shapeName=feather_small_mdRtA_capsule
   center=<-1.2978 -6.9399 -0.6926>
   rotation=<rot 0 162.4744 -81.9888>
  }]
 feather_small_mtRtB [{
   shapeType=Capsule
   length=8.7499
   radius=3.5765
   shapeName=feather_small_mtRtB_capsule
   center=<-1.4966 -6.837 4.5246>
   rotation=<rot 0 -11.3081 -50.081>
  }]
 feather_small_tpLfA [{
   shapeType=Capsule
   length=8.6842
   radius=2.5751
   shapeName=feather_small_tpLfA_capsule
   center=<0.9713 -6.0353 4.9179>
   rotation=<rot 0 7.3983 -54.1468>
  }]
 feather_small_tpRtA [{
   shapeType=Capsule
   length=9.7305
   radius=3.2356
   shapeName=feather_small_tpRtA_capsule
   center=<-0.9024 -6.3049 5.8004>
   rotation=<rot 0 -5.5844 -56.6987>
  }]
}
constraints [
 {bone1Name=feather_big_tpLfA, bone2Name=Melee_TwoHanded_Handle, swing1LimitDegrees=65, swing2LimitDegrees=65, twistLimitDegrees=65}
 {bone1Name=feather_big_mdLfA, bone2Name=Melee_TwoHanded_Handle, swing1LimitDegrees=65, swing2LimitDegrees=65, twistLimitDegrees=65}
 {bone1Name=feather_big_mdLfB, bone2Name=Melee_TwoHanded_Handle, swing1LimitDegrees=65, swing2LimitDegrees=65, twistLimitDegrees=65}
 {bone1Name=feather_big_mdCtA, bone2Name=Melee_TwoHanded_Handle, swing1LimitDegrees=65, swing2LimitDegrees=65, twistLimitDegrees=65}
 {bone1Name=feather_big_tpRtA, bone2Name=Melee_TwoHanded_Handle, swing1LimitDegrees=65, swing2LimitDegrees=65, twistLimitDegrees=65}
 {bone1Name=feather_big_mdRtA, bone2Name=Melee_TwoHanded_Handle, swing1LimitDegrees=65, swing2LimitDegrees=65, twistLimitDegrees=65}
 {bone1Name=feather_big_mdRtB, bone2Name=Melee_TwoHanded_Handle, swing1LimitDegrees=65, swing2LimitDegrees=65, twistLimitDegrees=65}
 {bone1Name=feather_big_btRtA, bone2Name=Melee_TwoHanded_Handle, swing1LimitDegrees=65, swing2LimitDegrees=65, twistLimitDegrees=65}
 {bone1Name=feather_big_btLfA, bone2Name=Melee_TwoHanded_Handle, swing1LimitDegrees=65, swing2LimitDegrees=65, twistLimitDegrees=65}
 {bone1Name=feather_small_btRtA, bone2Name=Melee_TwoHanded_Handle, swing1LimitDegrees=65, swing2LimitDegrees=65, twistLimitDegrees=65}
 {bone1Name=feather_small_btRtB, bone2Name=Melee_TwoHanded_Handle, swing1LimitDegrees=65, swing2LimitDegrees=65, twistLimitDegrees=65}
 {bone1Name=feather_small_mdRtA, bone2Name=Melee_TwoHanded_Handle, swing1LimitDegrees=65, swing2LimitDegrees=65, twistLimitDegrees=65}
 {bone1Name=feather_small_mtRtB, bone2Name=Melee_TwoHanded_Handle, swing1LimitDegrees=65, swing2LimitDegrees=65, twistLimitDegrees=65}
 {bone1Name=feather_small_tpRtA, bone2Name=Melee_TwoHanded_Handle, swing1LimitDegrees=65, swing2LimitDegrees=65, twistLimitDegrees=65}
 {bone1Name=feather_small_btLfA, bone2Name=Melee_TwoHanded_Handle, swing1LimitDegrees=65, swing2LimitDegrees=65, twistLimitDegrees=65}
 {bone1Name=feather_small_btLfB, bone2Name=Melee_TwoHanded_Handle, swing1LimitDegrees=65, swing2LimitDegrees=65, twistLimitDegrees=65}
 {bone1Name=feather_small_mdLfA, bone2Name=Melee_TwoHanded_Handle, swing1LimitDegrees=65, swing2LimitDegrees=65, twistLimitDegrees=65}
 {bone1Name=feather_small_mdLfB, bone2Name=Melee_TwoHanded_Handle, swing1LimitDegrees=65, swing2LimitDegrees=65, twistLimitDegrees=65}
 {bone1Name=feather_small_tpLfA, bone2Name=Melee_TwoHanded_Handle, swing1LimitDegrees=40, swing2LimitDegrees=40, twistLimitDegrees=40}
]
