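# Animation blueprint for the Polygon Male pickaxe that makes its big and small feather bones flutter.
# Each update, eight RandomRotatorFunction calls turn game time and Perlin settings into ambient rotators, scaled by separate big and small amplitude variables.
# The AnimGraph passes the input pose through a low-alpha, world-space RigidBody node, then chains additive bone-space rotations that share those rotators across the feathers.
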AnimBlueprint /BRCosmetics/Weapons/FORT_Melee/Pickaxe_Polygon_Male/Meshes/Polygon_Male_Axe_AnimBP
target_skeleton=&/Polygon_Male_Axe_Skeleton
graphs [
 AnimGraph {nodes [
   AnimGraphNode_SubInput_0 = Input Pose()
   n1 = RigidBody(ComponentPose=Local To Component(LocalPose=@AnimGraphNode_SubInput_0), Alpha=0.1) {
    overridePhysicsAsset=&/Polygon_Male_Axe_Physics
    simulationSpace=WorldSpace
    alphaBoolBlend {blendOption=Linear}
    baseBoneRef {boneName=Melee_TwoHanded_Handle}
   }
   n2 = Get ambientRotatorBigA() -> (ambientRotatorBigA:Rotator)
   n3 = "Transform (Modify) Bone - Bone: feather_big_tpLfA"(ComponentPose=@n1, Alpha=1, Translation="0,0,0", Rotation=@n2, Scale="1,1,1") {
    rotationMode=BMM_Additive
    rotationSpace=BCS_BoneSpace
    translationMode=BMM_Additive
    translationSpace=BCS_BoneSpace
    alphaBoolBlend {blendOption=Linear}
    boneToModify {boneName=feather_big_tpLfA}
   }
   n4 = Get ambientRotatorBigB() -> (ambientRotatorBigB:Rotator)
   n5 = "Transform (Modify) Bone - Bone: feather_big_mdLfA"(ComponentPose=@n3, Alpha=1, Translation="0,0,0", Rotation=@n4, Scale="1,1,1") {
    rotationMode=BMM_Additive
    rotationSpace=BCS_BoneSpace
    translationMode=BMM_Additive
    translationSpace=BCS_BoneSpace
    alphaBoolBlend {blendOption=Linear}
    boneToModify {boneName=feather_big_mdLfA}
   }
   n6 = "Transform (Modify) Bone - Bone: feather_big_mdLfB"(ComponentPose=@n5, Alpha=1, Translation="0,0,0", Rotation=@n2, Scale="1,1,1") {
    rotationMode=BMM_Additive
    rotationSpace=BCS_BoneSpace
    translationMode=BMM_Additive
    translationSpace=BCS_BoneSpace
    alphaBoolBlend {blendOption=Linear}
    boneToModify {boneName=feather_big_mdLfB}
   }
   n7 = Get ambientRotatorBigC() -> (ambientRotatorBigC:Rotator)
   n8 = "Transform (Modify) Bone - Bone: feather_big_mdCtA"(ComponentPose=@n6, Alpha=1, Translation="0,0,0", Rotation=@n7, Scale="1,1,1") {
    rotationMode=BMM_Additive
    rotationSpace=BCS_BoneSpace
    translationMode=BMM_Additive
    translationSpace=BCS_BoneSpace
    alphaBoolBlend {blendOption=Linear}
    boneToModify {boneName=feather_big_mdCtA}
   }
   n9 = Get ambientRotatorBigD() -> (ambientRotatorBigD:Rotator)
   n10 = "Transform (Modify) Bone - Bone: feather_big_tpRtA"(ComponentPose=@n8, Alpha=1, Translation="0,0,0", Rotation=@n9, Scale="1,1,1") {
    rotationMode=BMM_Additive
    rotationSpace=BCS_BoneSpace
    translationMode=BMM_Additive
    translationSpace=BCS_BoneSpace
    alphaBoolBlend {blendOption=Linear}
    boneToModify {boneName=feather_big_tpRtA}
   }
   n11 = "Transform (Modify) Bone - Bone: feather_big_mdRtA"(ComponentPose=@n10, Alpha=1, Translation="0,0,0", Rotation=@n4, Scale="1,1,1") {
    rotationMode=BMM_Additive
    rotationSpace=BCS_BoneSpace
    translationMode=BMM_Additive
    translationSpace=BCS_BoneSpace
    alphaBoolBlend {blendOption=Linear}
    boneToModify {boneName=feather_big_mdRtA}
   }
   n12 = "Transform (Modify) Bone - Bone: feather_big_mdRtB"(ComponentPose=@n11, Alpha=1, Translation="0,0,0", Rotation=@n9, Scale="1,1,1") {
    rotationMode=BMM_Additive
    rotationSpace=BCS_BoneSpace
    translationMode=BMM_Additive
    translationSpace=BCS_BoneSpace
    alphaBoolBlend {blendOption=Linear}
    boneToModify {boneName=feather_big_mdRtB}
   }
   n13 = "Transform (Modify) Bone - Bone: feather_big_btRtA"(ComponentPose=@n12, Alpha=1, Translation="0,0,0", Rotation=@n4, Scale="1,1,1") {
    rotationMode=BMM_Additive
    rotationSpace=BCS_BoneSpace
    translationMode=BMM_Additive
    translationSpace=BCS_BoneSpace
    alphaBoolBlend {blendOption=Linear}
    boneToModify {boneName=feather_big_btRtA}
   }
   n14 = "Transform (Modify) Bone - Bone: feather_big_btLfA"(ComponentPose=@n13, Alpha=1, Translation="0,0,0", Rotation=@n7, Scale="1,1,1") {
    rotationMode=BMM_Additive
    rotationSpace=BCS_BoneSpace
    translationMode=BMM_Additive
    translationSpace=BCS_BoneSpace
    alphaBoolBlend {blendOption=Linear}
    boneToModify {boneName=feather_big_btLfA}
   }
   n15 = Get ambientRotatorSmallA() -> (ambientRotatorSmallA:Rotator)
   n16 = "Transform (Modify) Bone - Bone: feather_small_btRtA"(ComponentPose=@n14, Alpha=1, Translation="0,0,0", Rotation=@n15, Scale="1,1,1") {
    rotationMode=BMM_Additive
    rotationSpace=BCS_BoneSpace
    translationMode=BMM_Additive
    translationSpace=BCS_BoneSpace
    alphaBoolBlend {blendOption=Linear}
    boneToModify {boneName=feather_small_btRtA}
   }
   n17 = Get ambientRotatorSmallB() -> (ambientRotatorSmallB:Rotator)
   n18 = "Transform (Modify) Bone - Bone: feather_small_btRtB"(ComponentPose=@n16, Alpha=1, Translation="0,0,0", Rotation=@n17, Scale="1,1,1") {
    rotationMode=BMM_Additive
    rotationSpace=BCS_BoneSpace
    translationMode=BMM_Additive
    translationSpace=BCS_BoneSpace
    alphaBoolBlend {blendOption=Linear}
    boneToModify {boneName=feather_small_btRtB}
   }
   n19 = Get ambientRotatorSmallC() -> (ambientRotatorSmallC:Rotator)
   n20 = "Transform (Modify) Bone - Bone: feather_small_mdRtA"(ComponentPose=@n18, Alpha=1, Translation="0,0,0", Rotation=@n19, Scale="1,1,1") {
    rotationMode=BMM_Additive
    rotationSpace=BCS_BoneSpace
    translationMode=BMM_Additive
    translationSpace=BCS_BoneSpace
    alphaBoolBlend {blendOption=Linear}
    boneToModify {boneName=feather_small_mdRtA}
   }
   n21 = Get ambientRotatorSmallD() -> (ambientRotatorSmallD:Rotator)
   n22 = "Transform (Modify) Bone - Bone: feather_small_mtRtB"(ComponentPose=@n20, Alpha=1, Translation="0,0,0", Rotation=@n21, Scale="1,1,1") {
    rotationMode=BMM_Additive
    rotationSpace=BCS_BoneSpace
    translationMode=BMM_Additive
    translationSpace=BCS_BoneSpace
    alphaBoolBlend {blendOption=Linear}
    boneToModify {boneName=feather_small_mtRtB}
   }
   n23 = "Transform (Modify) Bone - Bone: feather_small_btLfA"(ComponentPose=@n22, Alpha=1, Translation="0,0,0", Rotation=@n17, Scale="1,1,1") {
    rotationMode=BMM_Additive
    rotationSpace=BCS_BoneSpace
    translationMode=BMM_Additive
    translationSpace=BCS_BoneSpace
    alphaBoolBlend {blendOption=Linear}
    boneToModify {boneName=feather_small_btLfA}
   }
   n24 = "Transform (Modify) Bone - Bone: feather_small_btLfB"(ComponentPose=@n23, Alpha=1, Translation="0,0,0", Rotation=@n15, Scale="1,1,1") {
    rotationMode=BMM_Additive
    rotationSpace=BCS_BoneSpace
    translationMode=BMM_Additive
    translationSpace=BCS_BoneSpace
    alphaBoolBlend {blendOption=Linear}
    boneToModify {boneName=feather_small_btLfB}
   }
   n25 = "Transform (Modify) Bone - Bone: feather_small_mdLfA"(ComponentPose=@n24, Alpha=1, Translation="0,0,0", Rotation=@n19, Scale="1,1,1") {
    rotationMode=BMM_Additive
    rotationSpace=BCS_BoneSpace
    translationMode=BMM_Additive
    translationSpace=BCS_BoneSpace
    alphaBoolBlend {blendOption=Linear}
    boneToModify {boneName=feather_small_mdLfA}
   }
   n26 = "Transform (Modify) Bone - Bone: feather_small_mdLfB"(ComponentPose=@n25, Alpha=1, Translation="0,0,0", Rotation=@n17, Scale="1,1,1") {
    rotationMode=BMM_Additive
    rotationSpace=BCS_BoneSpace
    translationMode=BMM_Additive
    translationSpace=BCS_BoneSpace
    alphaBoolBlend {blendOption=Linear}
    boneToModify {boneName=feather_small_mdLfB}
   }
   n27 = "Transform (Modify) Bone - Bone: feather_small_tpLfA"(ComponentPose=@n26, Alpha=1, Translation="0,0,0", Rotation=@n21, Scale="1,1,1") {
    rotationMode=BMM_Additive
    rotationSpace=BCS_BoneSpace
    translationMode=BMM_Additive
    translationSpace=BCS_BoneSpace
    alphaBoolBlend {blendOption=Linear}
    boneToModify {boneName=feather_small_tpLfA}
   }
   n28 = "Transform (Modify) Bone - Bone: feather_small_tpRtA"(ComponentPose=@n27, Alpha=1, Translation="0,0,0", Rotation=@n17, Scale="1,1,1") {
    rotationMode=BMM_Additive
    rotationSpace=BCS_BoneSpace
    translationMode=BMM_Additive
    translationSpace=BCS_BoneSpace
    alphaBoolBlend {blendOption=Linear}
    boneToModify {boneName=feather_small_tpRtA}
   }
   Output Pose(Result=Component To Local(ComponentPose=@n28)) {layerGroup=None}
  ]}
 EventGraph {nodes [
   Event BlueprintUpdateAnimation()
   > n1 = Sequence!(execute) -> (then_0, then_1, then_2:Exec, then_3:Exec, then_4:Exec, then_5:Exec, then_6:Exec, then_7:Exec)
   n2 = GetGameTimeInSeconds()
   > n3 = RandomRotatorFunction!(
    execute
    GameTimeInSeconds:double=@n2
    PerlinScale:double=2
    Frequency:Rotator="5,3,4"
    Amplitude:Rotator="0.5,0.6,0.8"
    Offset:Rotator="0, 0, 0"
    TimeOffset:double=0
   ) -> (then, RandomRotator:Rotator)
   n4 = Get amplitudeScalarBig() -> (amplitudeScalarBig:double)
   > Set ambientRotatorBigA(ambientRotatorBigA:Rotator="Scale Rotator (Float)"!(A:Rotator=@n3, B:float=@n4) -> (ReturnValue:Rotator)) -> (Output_Get:Rotator)
   n5 = RandomRotatorFunction(execute=@n1.then_1, GameTimeInSeconds=@n2, PerlinScale=2.4, Frequency="4,5,3", Amplitude="0.9,0.3,0.5", Offset="0, 0, 0", TimeOffset=0)
   > Set ambientRotatorBigB(ambientRotatorBigB:Rotator="Scale Rotator (Float)"(A=@n5, B=@n4)) -> (Output_Get:Rotator)
   n6 = RandomRotatorFunction(execute=@n1.then_2, GameTimeInSeconds=@n2, PerlinScale=1.8, Frequency="2,4,2", Amplitude="0.6,0.7,0.8", Offset="0, 0, 0", TimeOffset=0)
   > Set ambientRotatorBigC(ambientRotatorBigC:Rotator="Scale Rotator (Float)"(A=@n6, B=@n4)) -> (Output_Get:Rotator)
   n7 = RandomRotatorFunction(execute=@n1.then_3, GameTimeInSeconds=@n2, PerlinScale=2.2, Frequency="2.5,3,3.5", Amplitude="0.7,0.8,0.6", Offset="0, 0, 0", TimeOffset=0)
   > Set ambientRotatorBigD(ambientRotatorBigD:Rotator="Scale Rotator (Float)"(A=@n7, B=@n4)) -> (Output_Get:Rotator)
   n8 = RandomRotatorFunction(execute=@n1.then_4, GameTimeInSeconds=@n2, PerlinScale=1, Frequency="7,6,8", Amplitude="0.6,0.5,0.8", Offset="0, 0, 0", TimeOffset=0)
   n9 = Get amplitudeScalarSmall() -> (amplitudeScalarSmall:double)
   > Set ambientRotatorSmallA(ambientRotatorSmallA:Rotator="Scale Rotator (Float)"(A=@n8, B=@n9)) -> (Output_Get:Rotator)
   n10 = RandomRotatorFunction(execute=@n1.then_5, GameTimeInSeconds=@n2, PerlinScale=0.8, Frequency="8,9,7", Amplitude="0.5,0.8,0.6", Offset="0, 0, 0", TimeOffset=0)
   > Set ambientRotatorSmallB(ambientRotatorSmallB:Rotator="Scale Rotator (Float)"(A=@n10, B=@n9)) -> (Output_Get:Rotator)
   n11 = RandomRotatorFunction(execute=@n1.then_6, GameTimeInSeconds=@n2, PerlinScale=1.1, Frequency="9,7.2,6.5", Amplitude="0.85,0.7,0.65", Offset="0, 0, 0", TimeOffset=0)
   > Set ambientRotatorSmallC(ambientRotatorSmallC:Rotator="Scale Rotator (Float)"(A=@n11, B=@n9)) -> (Output_Get:Rotator)
   n12 = RandomRotatorFunction(
    execute=@n1.then_7
    GameTimeInSeconds=@n2
    PerlinScale=0.9
    Frequency="5.5,6.8,7.2"
    Amplitude="0.71,0.9,0.82"
    Offset="0, 0, 0"
    TimeOffset=0
   )
   > Set ambientRotatorSmallD(ambientRotatorSmallD:Rotator="Scale Rotator (Float)"(A=@n12, B=@n9)) -> (Output_Get:Rotator)
  ]}
]
variables [
 ambientRotatorBigA {type=Rotator, default=<rot 5 5 5>}
 ambientRotatorBigB {type=Rotator}
 ambientRotatorBigC {type=Rotator}
 ambientRotatorBigD {type=Rotator}
 ambientRotatorSmallA {type=Rotator}
 ambientRotatorSmallB {type=Rotator}
 ambientRotatorSmallC {type=Rotator}
 ambientRotatorSmallD {type=Rotator}
 amplitudeScalarBig {type=number, default=1}
 amplitudeScalarSmall {type=number, default=1}
]
